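# Translucent, unlit, two-sided master material for lightning line effects on Niagara ribbons, sprites and mesh particles.
# Floor-stepped time jumps a noise texture, whose output distorts the panning base and mask textures, shaped by edge masks into opacity.
# Particle color drives emissive and alpha, with optional world-position sphere masking, view-angle fade and depth fade on some shading paths.
# SplineThicken supplies switchable width offset, while dynamic parameters allow per-particle noise, tiling, speed and width control.
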Material /Game/Effects/Niagara/MasterMaterials/Lightning/M_LineLightning
graph {
 properties {
  bUsedWithNiagaraMeshParticles=true
  bUsedWithNiagaraRibbons=true
  bUsedWithNiagaraSprites=true
  blendMode=BLEND_Translucent
  pixelDepthOffsetMode=PDOM_Legacy
  shadingModel=MSM_Unlit
  twoSided=true
  usedShadingModels=Unlit
 }
 nodes [
  n1 = ScalarParameter() {defaultValue=3, parameterName=SplineBase}
  n2 = DynamicParameter() {
   parameterIndex=3
   defaultValue=<rgba 1 2 1 200>
   paramNames [Noise_Intensity_offset, FloorNoise_SpeedOffset, CustomSplineWidth, WorldPosition_SphereMaskRadius]
  }
  n3 = LinearInterpolate(B=@n2.CustomSplineWidth, Alpha=ScalarParameter() {parameterName=Use Custom SplineWidth}) {constA=1}
  n4 = ScalarParameter() {defaultValue=3, parameterName=SplineTip}
  LinearInterpolate(
   A=MaterialFunctionCall(WidthBase=Multiply(A=@n1, B=@n3), WidthTip=Multiply(A=@n4, B=@n3), Expand U or V UV Channel=StaticBool()) {materialFunction=&SplineThicken}.WorldPositionOffset
   B=Constant()
   Alpha=ScalarParameter() {defaultValue=1, parameterName=Use_SplineThicken}
  )
  n5 = ParticleColor()
  MaterialFunctionCall(Emissive=@n5.RGB) {materialFunction=&MF_FX_EmissiveAdjustments}
  n6 = ComponentMask(None=TextureCoordinate() {vTiling=0.45}) {r=true}
  n7 = Saturate(None=Power(Base=Multiply(A=@n6) {constB=20}) {constExponent=1})
  n8 = OneMinus(None=@n6)
  n9 = Saturate(None=Power(Base=Multiply(A=@n8) {constB=20}) {constExponent=1})
  NamedRerouteDeclaration(None=Saturate(None=Multiply(A=@n7, B=@n9))) {name=Noisemask, nodeColor=<rgba 1 0.3473 0.2776 1>}
  n10 = Add(A=Multiply(A=Constant() {r=1}), B=VertexColor().G)
  n11 = Constant()
  n12 = VectorParameter() {group=VertexControl, parameterName=NoiseTilingSpeed, defaultValue=<rgba 1.2 0.8 -2 0>}
  n13 = Multiply(A=Add(A=TextureCoordinate(), B=AppendVector(A=@n10, B=@n11)), B=AppendVector(A=@n12.R, B=@n12.R))
  n14 = ComponentMask(None=@n13) {r=true}
  n15 = Time()
  n16 = ScalarParameter() {defaultValue=6, group=VertexFloorNoise, parameterName=VertexFloorNoise_Speed}
  n17 = Multiply(
   A=Floor(None=Multiply(A=@n15, B=@n16) {constB=6})
   B=ScalarParameter() {defaultValue=0.2, group=VertexFloorNoise, parameterName=VertesFloorNoise_MoveRate}
  ) {constB=0.2}
  n18 = ComponentMask(None=@n13) {g=true}
  NamedRerouteDeclaration(None=AppendVector(A=Add(A=@n14, B=@n17), B=Add(A=@n18, B=@n17))) {name=VertexControl, nodeColor=<rgba 0 0.4844 1 1>}
  n19 = Add(A=TextureCoordinate() {vTiling=0.8}, B=Constant2Vector() {r=0.15})
  n20 = DynamicParameter() {parameterIndex=2, defaultValue=<rgba 1 0 1 1>, paramNames [NoiseTextureU, NoiseTextureV, NoiseTextureU_offset, NoiseTextureV_offset]}
  n21 = Multiply(
   A=Add(A=TextureCoordinate(), B=AppendVector(A=@n20.NoiseTextureU_offset, B=@n20.NoiseTextureU_offset))
   B=AppendVector(A=@n20.NoiseTextureU, B=@n20.NoiseTextureU)
  )
  n22 = Time()
  n23 = DynamicParameter() {
   parameterIndex=3
   defaultValue=<rgba 1 2 1 200>
   paramNames [Noise_Intensity_offset, FloorNoise_SpeedOffset, CustomSplineWidth, WorldPosition_SphereMaskRadius]
  }
  n24 = Multiply(A=@n23.FloorNoise_SpeedOffset, B=ScalarParameter() {defaultValue=6, group=FloorNoise, parameterName=FloorNoise_Speed})
  n25 = Multiply(
   A=Floor(None=Multiply(A=@n22, B=@n24) {constB=6})
   B=ScalarParameter() {defaultValue=0.2, group=FloorNoise, parameterName=FloorNoise_MoveRate}
  ) {constB=0.2}
  n26 = Add(A=ComponentMask(None=@n21) {r=true}, B=@n25)
  n27 = Add(A=ComponentMask(None=@n21) {g=true}, B=@n25)
  n28 = TextureSampleParameter2D(UVs=LinearInterpolate(
    A=NamedRerouteUsage()
    B=AppendVector(A=@n26, B=@n27)
    Alpha=ScalarParameter() {defaultValue=1, group=VertexControl, parameterName=Use_VertexControl}
   )) {group=Noise, parameterName=NoiseTexture, texture=/Game/Weapons/FORT_Melee/Pickaxe_CoatCheck/FX/T_Pickaxe_CoatCheck_Noise}
  n29 = NamedRerouteUsage()
  n30 = LinearInterpolate(A=ScalarParameter() {defaultValue=0.1, group=Noise, parameterName=NoiseTexMaskLerp}, Alpha=ComponentMask(None=TextureCoordinate()) {r=true}) {constA=0.1}
  n31 = ScalarParameter() {defaultValue=1, group=Noise, parameterName=Use_NoiseMaskEdge}
  n32 = Multiply(
   A=Multiply(A=ComponentMask(None=@n28.RGB) {g=true, r=true}, B=LinearInterpolate(A=@n29, B=@n30, Alpha=@n31))
   B=Multiply(A=ScalarParameter() {defaultValue=0.4, group=Noise, parameterName=Noise_Intensity}, B=@n23.Noise_Intensity_offset)
  ) {constB=0.3}
  n33 = ScalarParameter() {defaultValue=-1, group=BaseTexture, parameterName=BaseTexture_SpeedX}
  n34 = ScalarParameter() {defaultValue=-0.1, group=BaseTexture, parameterName=BaseTexture_SpeedY}
  n35 = TextureSampleParameter2D(UVs=Panner(Coordinate=Add(A=@n19, B=@n32), Speed=AppendVector(A=@n33, B=@n34)) {speedX=-1, speedY=-0.1}) {
   group=BaseTexture
   parameterName=Base_Texture
   samplerType=SAMPLERTYPE_LinearGrayscale
   texture=/Game/Environments/Hexmap/HexClouds/T_HexmapFog_01_Height02
  }
  n36 = ChannelMaskParameter(None=@n35.RGBA) {group=BaseTexture, parameterName=Tex_ChargeRGBA, sortPriority=7}
  n37 = Panner(Coordinate=TextureCoordinate() {uTiling=2, vTiling=0.8}) {speedX=-2}
  n38 = Add(
   A=TextureCoordinate() {vTiling=0.45}
   B=Multiply(
    A=TextureSample(UVs=@n37) {samplerType=SAMPLERTYPE_LinearGrayscale, texture=&T_HexmapFog_01_Height02}.R
    B=Constant() {r=0.03}
   ) {constB=0.3}
  )
  n39 = ComponentMask(None=@n38) {r=true}
  n40 = ScalarParameter() {defaultValue=20, group=LinearMask, parameterName=MaskLeft_Mul}
  n41 = Saturate(None=Power(Base=Multiply(A=@n39, B=@n40) {constB=20}) {constExponent=1})
  n42 = OneMinus(None=@n39)
  n43 = ScalarParameter() {defaultValue=2.5, group=LinearMask, parameterName=MaskRight_Mul}
  n44 = Saturate(None=Power(Base=Multiply(A=@n42, B=@n43) {constB=2.5}) {constExponent=1})
  n45 = ComponentMask(None=@n38) {g=true}
  n46 = Power(Base=Multiply(A=Sine(None=@n45), B=ScalarParameter() {defaultValue=1, group=LinearMask, parameterName=UpDownMask_Mul})) {constExponent=3}
  n47 = LinearInterpolate(A=ScalarParameter() {group=LinearMask, parameterName=LinearMaskLerp}, Alpha=Multiply(A=Multiply(A=@n41, B=@n44), B=Saturate(None=@n46)))
  n48 = DynamicParameter() {defaultValue=<rgba 1 1 1 0>, paramNames [MaskTextureU_offset, MaskTextureV_Offset, MaskTextureSpeedX_offset, MaskTextureSpeedY_offset]}
  n49 = Multiply(A=ScalarParameter() {defaultValue=1, group=MaskTexture, parameterName=MaskTextureU}, B=@n48.MaskTextureU_offset)
  n50 = Multiply(A=ScalarParameter() {defaultValue=1, group=MaskTexture, parameterName=MaskTextureV}, B=@n48.MaskTextureV_Offset)
  n51 = Add(A=Multiply(A=TextureCoordinate(), B=AppendVector(A=@n49, B=@n50)), B=@n32)
  n52 = AppendVector(
   A=Multiply(A=ScalarParameter() {defaultValue=-1, group=MaskTexture, parameterName=MaskTexture_SpeedX}, B=@n48.MaskTextureSpeedX_offset)
   B=Multiply(A=ScalarParameter() {group=MaskTexture, parameterName=MaskTexture_SpeedY}, B=@n48.MaskTextureSpeedY_offset)
  )
  n53 = Multiply(
   A=Power(Base=Multiply(A=@n36.R, B=@n47), Exp=LinearInterpolate() {constA=2})
   B=TextureSampleParameter2D(UVs=Panner(Coordinate=@n51, Speed=@n52) {speedX=-1}) {group=MaskTexture, parameterName=MaskTexture, texture=/Game/Weapons/FORT_Rifles/Effects/Textures/T_FX_ThinLine_BoomB_01}.R
  )
  n54 = Multiply(A=@n5.A, B=Power(Base=LinearInterpolate(Alpha=@n53) {constB=2.5}, Exp=ScalarParameter() {defaultValue=1, parameterName=Power}))
  n55 = AppendVector(A=ScalarParameter() {parameterName=ParticlePosition_X}, B=ScalarParameter() {parameterName=ParticlePosition_Y})
  n56 = ScalarParameter() {parameterName=ParticlePosition_Z}
  n57 = DynamicParameter() {
   parameterIndex=3
   defaultValue=<rgba 1 2 1 200>
   paramNames [Noise_Intensity_offset, FloorNoise_SpeedOffset, CustomSplineWidth, WorldPosition_SphereMaskRadius]
  }
  n58 = SphereMask(A=WorldPosition().XYZ, B=Add(A=ParticlePositionWS(), B=AppendVector(A=@n55, B=@n56)), Radius=@n57.WorldPosition_SphereMaskRadius) {attenuationRadius=65, hardnessPercent=80}
  n59 = Multiply(
   A=LinearInterpolate(A=Multiply(A=@n54, B=@n58), B=@n54, Alpha=ScalarParameter() {defaultValue=1, parameterName=Use_WorldPositionSphereMask})
   B=MaterialFunctionCall("FadeSensitivity (optional)"=ScalarParameter() {parameterName=MeshFade}) {materialFunction=&FadeBasedOnViewAngle}.Result
  )
  n60 = DepthFade(Opacity=@n59, FadeDistance=ScalarParameter() {defaultValue=5, parameterName=DepthFade})
  ShadingPathSwitch(Default=@n60, Deferred=@n60, Forward=@n59, Mobile=@n59)
 ]
}
